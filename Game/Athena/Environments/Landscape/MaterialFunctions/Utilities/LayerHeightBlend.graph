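# Landscape material function that height-blends a Top material attribute layer over a Bottom one using an Alpha input.
# It reads each layer's Opacity as height, scales the Top-minus-Bottom difference by Height Multiplier, and remaps it into HeightLerp with a Contrast control.
# The resulting mask drives BlendMaterialAttributes.
# Top's BaseColor is multiplied by an Edge Tint Color that fades out as Alpha rises, tinting the transition.
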MaterialFunction /Game/Athena/Environments/Landscape/MaterialFunctions/Utilities/LayerHeightBlend
graph {nodes [
  n1 = GetMaterialAttributes(None=FunctionInput() {inputName=Bottom, inputType=FunctionInput_MaterialAttributes}) {attributeGetTypes [#1]}
  n2 = GetMaterialAttributes(None=FunctionInput() {inputName=Top, inputType=FunctionInput_MaterialAttributes, sortPriority=1}) {attributeGetTypes [#1, #2]}
  n3 = VectorParameter() {parameterName=TintColor, defaultValue=<rgba 1 1 1 1>}
  n4 = Saturate(None=FunctionInput() {inputName=Alpha, inputType=FunctionInput_Scalar, sortPriority=2})
  n5 = Multiply(
   A=@n2.BaseColor
   B=LinearInterpolate(A=FunctionInput(Preview=@n3.RGB) {bUsePreviewValueAsDefault=true, inputName=Edge Tint Color, sortPriority=3}, Alpha=@n4)
  )
  n6 = FunctionInput() {bUsePreviewValueAsDefault=true, inputName=Height Multiplier, inputType=FunctionInput_Scalar, sortPriority=5, previewValue=<1 0 0 1>}
  n7 = ConstantBiasScale(None=Multiply(A=Subtract(A=@n2.Opacity, B=@n1.Opacity), B=@n6))
  n8 = FunctionInput(Preview=ScalarParameter() {defaultValue=1, parameterName=Preview Height Blend Contrast}) {bUsePreviewValueAsDefault=true, inputName=Contrast, inputType=FunctionInput_Scalar, sortPriority=5}
  FunctionOutput(None=BlendMaterialAttributes(
    A=@n1.MaterialAttributes
    B=SetMaterialAttributes!(MaterialAttributes=@n2.MaterialAttributes, BaseColor=@n5) {attributeSetTypes [#2]}
    Alpha=MaterialFunctionCall(Transition Phase=@n4, Height Texture=@n7, Contrast=@n8) {materialFunction=&HeightLerp}.Alpha
   ))
 ]}
