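# Landscape material for a default Creative terrain biome, blending cliff rock, grass and gravel.
# Vertex-normal slope drives rock-to-gravel transitions in world-aligned textures and height-blends grass over them, with grass tiling changing by distance.
# A painted Gravel layer height-blends on top before the result passes through MF_Curie_Landscape.
# Landscape grass output spawns grass on flat ground minus a NoFoliage layer, plus flowers from a painted layer.
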Material /Game/Athena/Environments/Landscape/Creative/M_Athena_Terrain_DefaultBiome_02
graph {
 properties {bUseMaterialAttributes=true, pixelDepthOffsetMode=PDOM_Legacy, usedShadingModels=Default Lit}
 nodes [
  n1 = ComponentMask(None=VertexNormalWS()) {b=true}
  n2 = ScalarParameter() {defaultValue=0.85, parameterName=LS Grass Threshold}
  MaterialExpressionSubtract_10 = Subtract(A=Ceil(None=Subtract(A=@n1, B=@n2)), B=LandscapeLayerSample() {parameterName=NoFoliage})
  MaterialExpressionLandscapeLayerSample_8 = LandscapeLayerSample() {parameterName=Flowers}
  LandscapeGrassOutput!(Grass=@MaterialExpressionSubtract_10, Flowers=@MaterialExpressionLandscapeLayerSample_8) {grassTypes [
    Grass {
     grassType=/Game/Athena/Environments/Landscape/Island/Foliage/Athena_Foliage_Grass_CreativeIslands
     input {
      expression=&self:MaterialExpressionSubtract_10
      inputName=None
      mask=0
      maskA=0
      maskB=0
      maskG=0
      maskR=0
      outputIndex=0
     }
    }
    Flowers {
     grassType=/Game/Athena/Environments/Landscape/Foliage/Athena_Foliage_Flowers
     input {
      expression=&self:MaterialExpressionLandscapeLayerSample_8
      inputName=None
      mask=0
      maskA=0
      maskB=0
      maskG=0
      maskR=0
      outputIndex=0
     }
    }
   ]}
  n3 = TextureObject() {texture=/Game/Packages/DS_Fortnite_Terrain_NoLOD/Textures/T_Cave_Exterior_Wall_D}
  n4 = Constant() {r=1534.9661}
  n5 = StaticBool() {value=true}
  n6 = NamedRerouteUsage()
  n7 = MaterialFunctionCall() {materialFunction=/Game/Athena/Environments/Landscape/MaterialFunctions/Standard/MF_Athena_RoadGravel}
  n8 = AppendVector(A=@n7.Gravel, B=@n7.Gravel)
  n9 = Constant() {r=1.75}
  n10 = ComponentMask(None=Multiply(A=VertexNormalWS(), B=Constant() {r=0.725})) {b=true}
  n11 = ComponentMask(None=MaterialFunctionCall(In=Abs(None=@n10), Contrast=NamedRerouteUsage()) {materialFunction=&CheapContrast}.Result) {r=true}
  n12 = Divide(A=LinearInterpolate(
    A=MaterialFunctionCall(TextureObject=@n3, TextureSize=@n4, Export Float 4=@n5, ProjectionTransitionContrast=@n6) {materialFunction=&WorldAlignedTexture}.XY Texture
    B=Multiply(A=@n8, B=@n9)
    Alpha=@n11
   )) {constB=1.25, desc=Darken Shallow rock to make up for no normal}
  n13 = Constant() {r=4500}
  n14 = NamedRerouteUsage()
  n15 = MaterialFunctionCall(TextureObject=@n3, TextureSize=@n13, Export Float 4=StaticBool() {value=true}, ProjectionTransitionContrast=@n14) {materialFunction=&WorldAlignedTexture}
  n16 = Constant() {r=5}
  n17 = Constant() {r=-4}
  n18 = ComponentMask(None=VertexNormalWS()) {b=true}
  n19 = Clamp(None=Add(A=LinearInterpolate(A=@n16, B=@n17, Alpha=@n18), B=Constant() {r=0.8}))
  n20 = Saturate(None=Divide(A=LinearInterpolate(A=@n12, B=@n15.XY Texture, Alpha=@n19)) {constB=1.5})
  n21 = LinearInterpolate(A=Constant3Vector() {constant=<rgba 0 0 1 0>}, B=@n7.GravelAndMudNormals, Alpha=@n11)
  n22 = MaterialFunctionCall!(
   TextureObject=TextureObject() {samplerType=SAMPLERTYPE_Normal, texture=/Game/Packages/DS_Fortnite_Terrain_NoLOD/Textures/T_Cave_Exterior_Wall_N}
   TextureSize=@n13
   WorldSpace
   Normal
   ProjectionTransitionContrast=Multiply(A=@n14, B=Constant() {r=-1})
   WorldPosition
   Use High Quality Normals
  ) {materialFunction=/Game/Athena/Environments/Materials/WorldAlignedNormalContrast}
  n23 = Constant3Vector() {constant=<rgba 0 0 1 0>}
  n24 = SetMaterialAttributes!(
   MaterialAttributes
   Base Color=ComponentMask(None=@n20) {b=true, g=true, r=true}
   Specular=Constant() {r=0.5}
   Roughness=Constant() {r=0.9}
   Opacity=ComponentMask(None=@n20) {a=true}
   Normal=ShadingPathSwitch(Default=LinearInterpolate(A=@n21, B=@n22.XY Texture, Alpha=@n19), Mobile=LinearInterpolate(A=@n23, B=@n22.XY Texture, Alpha=@n19)) {desc=Migrated from FeatureLevelSwitch}
  ) {attributeSetTypes [
    #1
    #2
    #3
    #4
    #5
   ]}
  n25 = ComponentMask(None=WorldPosition() {worldPositionShaderOffset=WPT_ExcludeAllShaderOffsets}.XYZ) {g=true, r=true}
  n26 = TextureSample(UVs=Divide(A=@n25, B=Constant() {r=700})) {samplerSource=SSM_Wrap_WorldGroupSettings, texture=/Game/Athena/Environments/Landscape/Textures/T_Athena_Grass_Farm_ColorMatched_D_2}
  n27 = TextureSample(UVs=Divide(A=@n25, B=Constant() {r=3000})) {samplerSource=SSM_Wrap_WorldGroupSettings, texture=&T_Athena_Grass_Farm_ColorMatched_D_2}
  n28 = Constant() {r=55000}
  n29 = Constant() {r=45000}
  n30 = MaterialFunctionCall!(Distance=ShadingPathSwitch(Default=QualitySwitch(Default=@n28, Low=@n29, High=@n28, Medium=@n28), Mobile=@n29) {desc=Migrated from FeatureLevelSwitch}) {materialFunction=/Game/Athena/Environments/Landscape/MaterialFunctions/MF_TerrainDistanceFade}
  n31 = LinearInterpolate(A=@n26.A, B=@n27.A, Alpha=@n30.Result)
  n32 = Constant() {r=2}
  n33 = SetMaterialAttributes!(
   MaterialAttributes
   Base Color=LinearInterpolate(A=@n26.RGB, B=@n27.RGB, Alpha=@n30.Result)
   Specular=LinearInterpolate(B=Constant() {r=0.15}, Alpha=Power(Base=@n31, Exp=@n32))
   Roughness=Constant() {r=0.9}
   Normal=Constant3Vector() {constant=<rgba 0 0 1 0>}
   Opacity=@n31
  ) {attributeSetTypes [
    #1
    #2
    #3
    #5
    #4
   ]}
  n34 = Multiply(A=ComponentMask(None=VertexNormalWS()) {b=true}, B=Constant() {r=0.65})
  n35 = Constant() {r=0.2}
  n36 = Constant() {r=4}
  n37 = Constant3Vector() {constant=<rgba 0.28 0.2437 0.21 1>}
  n38 = MaterialFunctionCall() {materialFunction=&MF_Athena_RoadGravel}
  n39 = Constant3Vector() {constant=<rgba 0.36 0.3902 0.4 1>}
  n40 = Constant() {r=2}
  n41 = SetMaterialAttributes(
   Base Color=LinearInterpolate(
    A=Multiply(A=@n37, B=@n38.Gravel_NoColor)
    B=Multiply(A=@n38.Gravel_NoColor, B=@n39)
    Alpha=MaterialFunctionCall(In=@n38.Gravel_NoColor, Contrast=@n40) {materialFunction=&CheapContrast}.Result
   )
   Specular=Constant() {r=0.5}
   Roughness=Constant() {r=1}
   Normal=@n38.GravelAndMudNormals
   Opacity=@n38.GravelAndMudNormals
  ) {attributeSetTypes [
    #1
    #2
    #3
    #5
    #4
   ]}
  n42 = GetMaterialAttributes(None=MaterialFunctionCall!(
    Bottom=MaterialFunctionCall(Bottom=@n24, Top=@n33, Alpha=@n34, Height Multiplier=@n35, Contrast=@n36) {materialFunction=/Game/Athena/Environments/Landscape/MaterialFunctions/Utilities/LayerHeightBlend}.Result
    Top=@n41
    Alpha=LandscapeLayerSample() {parameterName=Gravel}
    Edge Tint Color
    Height Multiplier=ScalarParameter() {defaultValue=2, parameterName=Gravel Height}
    Contrast=ScalarParameter() {defaultValue=1, parameterName=Gravel Contrast}
   ) {materialFunction=&LayerHeightBlend}.Result) {attributeGetTypes [
    #1
    #2
    #3
    #5
   ]}
  n43 = SetMaterialAttributes!(MaterialAttributes, Base Color=@n42.Base Color, Specular=@n42.Base Color, Roughness=@n42.Base Color, Normal=@n42.Base Color) {attributeSetTypes [
    #1
    #2
    #3
    #5
   ]}
  MaterialFunctionCall!(
   In=@n43
   UseHeightFalloff=StaticBool() {desc=Do Not UseHeightFalloff here}
   OnFireEmissiveOverride
   HeightFalloffRadius
   HeightFalloffHardness
   VariationMasksMobile
   In_Grass
   Is_Stylised_WE
  ) {desc=This should be near very last, materialFunction=/Game/Packages/Fortress_SharedMaterials/Base_Material_Functions/MF_Curie_Landscape}
  NamedRerouteDeclaration(None=Constant() {r=10}) {name=ProjCont, nodeColor=<rgba 1 0 1 1>}
 ]
}
