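# Desert biome terrain material for Creative islands, with an IsStaticmesh switch for landscape or mesh use.
# World-aligned cliff rock tinted by height-based sediment bands height-blends by slope with distance-faded grass, then gravel (Forest layer or vertex red) and wavy sand (Mud layer or vertex green).
# A landscape grass output spawns desert grass on flat ground, minus Mud, NoFoliage and Forest layers.
# Quality and shading-path switches simplify normals and sand on cheaper targets.
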Material /Game/Athena/Environments/Landscape/Creative/M_Athena_Terrain_DesertBiome1
graph {
 properties {bUseMaterialAttributes=true, pixelDepthOffsetMode=PDOM_Legacy, usedShadingModels=Default Lit}
 nodes [
  Constant3Vector() {desc=Use this value for Oak_Island_105x105, constant=<rgba 0.974 0.5815 0.2333 1>}
  Constant() {r=7}
  Constant() {r=0.1}
  Constant() {r=-1.2436}
  Constant() {r=1}
  Constant() {r=5}
  Constant() {r=0.1}
  Constant() {r=0.55}
  n1 = VertexNormalWS()
  n2 = Ceil(None=Subtract(A=ComponentMask(None=@n1) {b=true}, B=ScalarParameter() {defaultValue=0.95, parameterName=LS Grass Threshold}))
  n3 = Multiply(A=LandscapeLayerSample() {parameterName=Mud}) {constB=2}
  MaterialExpressionSubtract_2 = Subtract(
   A=Subtract(A=Subtract(A=@n2, B=@n3), B=LandscapeLayerSample() {parameterName=NoFoliage})
   B=Multiply(B=LandscapeLayerSample() {parameterName=Forest}) {constA=1.2}
  )
  LandscapeGrassOutput!(Grass=@MaterialExpressionSubtract_2) {grassTypes [Grass {
     grassType=/Game/Athena/Environments/Landscape/Island/Foliage/Athena_Foliage_DesertGrass_CreativeIslands
     input {
      expression=&self:MaterialExpressionSubtract_2
      inputName=None
      mask=0
      maskA=0
      maskB=0
      maskG=0
      maskR=0
      outputIndex=0
     }
    }]}
  n4 = MaterialFunctionCall!(
   Diffuse Texture Object=TextureObject() {samplerType=SAMPLERTYPE_LinearColor, texture=/Game/Athena/Environments/Landscape/Textures/Arid/T_Athena_Arid_Rock_01_M}
   Spec Texture Object
   TextureSize=Constant() {r=2500}
   --------------------------
   Use seperate settings for Z projection
   Use seperate texture for Z projection
   Texture Size Z projection
   Z Plane Diffuse Texture Object
   Z Plane Spec Texture Object
   "--------------------------      "
   ProjectionTransitionContrast=Constant() {r=10}
   Use Texture Alpha Heightmap to Refine the Transistion
   HeightLerpTransitionContrast
   PickTheDiffuseHeightChannel
   TheDiffuseHeightChannel
   "--------------------------       "
   NormalTextureObject=TextureObject() {samplerType=SAMPLERTYPE_Normal, texture=/Game/Environments/Sets/Arid/Textures/RockCliff_normal}
   World Space Normals=StaticBool()
   "--------------------------  "
   Export Float 4 Diffuse
   Export Float 4 Spec
   "--------------------------     "
   WorldPosition
   World Space Normal
  ) {materialFunction=/Engine/Functions/Engine_MaterialFunctions02/Texturing/WorldAlignedTextures_Complex}
  n5 = NamedRerouteUsage()
  n6 = LinearInterpolate(A=ComponentMask(None=@n4.XY Diffuse Texture) {r=true}, B=ComponentMask(None=@n4.XY Diffuse Texture) {g=true}, Alpha=@n5)
  n7 = ComponentMask(None=WorldPosition().XYZ) {b=true}
  n8 = Constant() {r=8000}
  n9 = TextureSample(UVs=Divide(A=Add(A=@n7, B=@n8), B=Constant() {r=4000})) {samplerSource=SSM_Wrap_WorldGroupSettings, texture=/Game/Environments/Sets/Arid/Textures/GradientSediments__M}
  n10 = Constant3Vector() {constant=<rgba 0.5 0.5 0.5 0>}
  n11 = Saturate(None=Multiply(A=Add(A=@n9.RGB, B=@n10), B=Constant3Vector() {constant=<rgba 0.75 0.6011 0.5586 0>}))
  n12 = SetMaterialAttributes!(
   MaterialAttributes
   Base Color=Multiply(A=Multiply(A=@n6, B=@n11), B=VectorParameter() {parameterName=CliffsBiomeColor, defaultValue=<rgba 0.7656 0.3161 0.162 1>}.RGB)
   Specular=Constant() {r=0.5}
   Roughness=Constant() {r=0.8}
   Normal=LinearInterpolate(A=@n4.Normal XY Texture, B=Constant3Vector() {constant=<rgba 0 0 1 0>}, Alpha=@n5)
   Opacity=@n6
  ) {attributeSetTypes [
    #1
    #2
    #3
    #4
    #5
   ]}
  n13 = ComponentMask(None=WorldPosition() {worldPositionShaderOffset=WPT_ExcludeAllShaderOffsets}.XYZ) {g=true, r=true}
  n14 = TextureSample(UVs=Divide(A=@n13, B=Constant() {r=700})) {samplerSource=SSM_Wrap_WorldGroupSettings, texture=/Game/Environments/Sets/Transylvania/Terrain/T_TRV_Grass_01_D}
  n15 = TextureSample(UVs=Divide(A=@n13, B=Constant() {r=3000})) {samplerSource=SSM_Wrap_WorldGroupSettings, texture=&T_TRV_Grass_01_D}
  n16 = Constant() {r=55000}
  n17 = Constant() {r=45000}
  n18 = MaterialFunctionCall!(Distance=ShadingPathSwitch(Default=QualitySwitch(Default=@n16, Low=@n17, High=@n16, Medium=@n16), Mobile=@n17) {desc=Migrated from FeatureLevelSwitch}) {materialFunction=/Game/Athena/Environments/Landscape/MaterialFunctions/MF_TerrainDistanceFade}
  n19 = LinearInterpolate(A=@n14.A, B=@n15.A, Alpha=@n18.Result)
  n20 = Constant() {r=2}
  n21 = SetMaterialAttributes(
   Base Color=LinearInterpolate(A=@n14.RGB, B=@n15.RGB, Alpha=@n18.Result)
   Specular=LinearInterpolate(B=Constant() {r=0.15}, Alpha=Power(Base=@n19, Exp=@n20))
   Roughness=Constant() {r=0.9}
   Normal=Constant3Vector() {constant=<rgba 0 0 1 0>}
   Opacity=@n19
  ) {attributeSetTypes [
    #1
    #2
    #3
    #4
    #5
   ]}
  n22 = VertexNormalWS()
  n23 = MaterialFunctionCall!(
   Bottom=@n12
   Top=@n21
   Alpha=Multiply(A=ComponentMask(None=@n22) {b=true}, B=ScalarParameter() {defaultValue=0.55, group=Rock, parameterName=RockMultiply Z Multiply})
   Edge Tint Color
   Height Multiplier=ScalarParameter() {defaultValue=0.1, parameterName=GrassRock Height Blend Mult}
   Contrast=ScalarParameter() {defaultValue=5, group=Grass, parameterName=GrassRock Contrast}
  ) {materialFunction=/Game/Athena/Environments/Landscape/MaterialFunctions/Utilities/LayerHeightBlend}
  n24 = VertexColor()
  n25 = NamedRerouteUsage()
  MaterialFunctionCall(
   Bottom=@n23.Result
   Top=NamedRerouteUsage()
   Alpha=StaticSwitchParameter(True=Multiply(A=@n24.R, B=@n25), False=LandscapeLayerSample() {parameterName=Forest}) {parameterName=IsStaticmesh}
   Edge Tint Color=VectorParameter() {parameterName=Forest_Tint, defaultValue=<rgba 1.5 1.5 1.5 0>}.RGB
   Height Multiplier=ScalarParameter() {defaultValue=7, parameterName=Forest Heightblend Multiplier}
   Contrast=ScalarParameter() {defaultValue=0.1, parameterName=Forest Heightblend Contrast}
  ) {materialFunction=&LayerHeightBlend}
  n26 = ScalarParameter() {defaultValue=-1.2436, group=Rock, parameterName=CliffBlend}
  n27 = ScalarParameter() {defaultValue=1, group=Rock, parameterName=CliffBlend2}
  n28 = ComponentMask(None=VertexNormalWS()) {b=true}
  NamedRerouteDeclaration(None=Saturate(None=LinearInterpolate(A=@n26, B=@n27, Alpha=@n28))) {name=BlendTopToCliff, nodeColor=<rgba 1 0 1 1>}
  n29 = ScalarParameter() {defaultValue=1, parameterName=NoiseLOW}
  n30 = ScalarParameter() {defaultValue=-4, parameterName=NoiseHigh}
  n31 = ScalarParameter() {defaultValue=11348.25, parameterName=NoiseSize}
  n32 = AppendVector(A=@n31, B=@n31)
  n33 = ComponentMask(None=MaterialFunctionCall(
    TextureObject=TextureObject() {samplerType=SAMPLERTYPE_Grayscale, texture=/Game/Global/Textures/Default/Noise/TilingNoise06}
    TextureSize=AppendVector(A=@n32, B=@n31)
   ) {materialFunction=&WorldAlignedTexture}.XY Texture) {r=true}
  NamedRerouteDeclaration(None=Saturate(None=LinearInterpolate(A=@n29, B=@n30, Alpha=@n33))) {name=CliffNoiseMask, nodeColor=<rgba 1 0 1 1>}
  n34 = GetMaterialAttributes(None=@n23.Result) {attributeGetTypes [#5]}
  n35 = ComponentMask(None=WorldPosition().XYZ) {g=true, r=true}
  n36 = ScalarParameter() {defaultValue=700, parameterName=SizeGravelDesert}
  n37 = TextureSample(UVs=StaticSwitchParameter(True=Divide(A=@n35, B=@n36), False=LandscapeLayerCoords() {mappingScale=4}) {parameterName=IsStaticmesh}) {samplerSource=SSM_Wrap_WorldGroupSettings, texture=/Game/Athena/Environments/Landscape/Textures/Arid/T_Athena_Arid_Gravel_D}
  n38 = Transform(None=MaterialFunctionCall!(Height=@n37.A, Absolute World Position, World Space Vertex Normals) {materialFunction=/Engine/Functions/Engine_MaterialFunctions03/Procedurals/HeightToNormalSmooth}.WorldSpaceNormal) {transformSourceType=TRANSFORMSOURCE_World, transformType=TRANSFORM_Tangent}
  n39 = Constant3Vector() {constant=<rgba 3 3 1 1>}
  n40 = SetMaterialAttributes(
   Base Color=Multiply(A=Constant() {r=0.8}, B=@n37.RGB)
   Specular=Constant() {r=0.2}
   Roughness=Constant() {r=1}
   Normal=QualitySwitch(Default=Constant3Vector() {constant=<rgba 0 0 1 0>}, Epic=Multiply(A=@n38, B=@n39))
   Opacity=@n37.A
  ) {attributeSetTypes [
    #1
    #2
    #3
    #4
    #5
   ]}
  n41 = Power(Base=@n37.A, Exp=Constant() {r=2})
  n42 = Constant() {r=4}
  n43 = Constant() {r=-5}
  n44 = Constant() {r=1}
  n45 = ComponentMask(None=VertexNormalWS()) {b=true}
  n46 = Multiply(A=Saturate(None=Multiply(A=@n41, B=@n42)), B=Saturate(None=LinearInterpolate(A=@n43, B=@n44, Alpha=@n45)))
  n47 = StaticSwitchParameter(True=Multiply(A=VertexColor().R, B=NamedRerouteUsage()), False=LandscapeLayerSample() {parameterName=Forest}) {parameterName=IsStaticmesh}
  n48 = MaterialFunctionCall!(
   Material 1=@n34.MaterialAttributes
   Height Texture 1=@n34.MaterialAttributes
   Material 2=NamedRerouteDeclaration(None=@n40) {name=GravelRocks, nodeColor=<rgba 1 0 1 1>}
   Height Texture 2=@n37.A
   Transistion Phase=Saturate(None=Add(A=@n46, B=@n47))
   UseContrast
   Contrast
   AddLayerMask
  ) {materialFunction=/Game/Athena/Apollo/Environments/Landscape/Materials/Biome/CommonMaterials/MF_BlendLandscapeLayers}
  n49 = TextureObject() {samplerType=SAMPLERTYPE_Masks, texture=/Game/Environments/Sets/Arid/Textures/T_desert_sand_wavy_S}
  n50 = ScalarParameter() {defaultValue=650, parameterName=SizeSandDesert}
  n51 = TextureObject() {samplerType=SAMPLERTYPE_Normal, texture=/Game/Environments/Sets/Arid/Textures/T_desert_sand_wavy_N}
  n52 = MaterialFunctionCall(
   Diffuse Texture Object=TextureObject() {texture=/Game/Environments/Sets/Arid/Textures/T_desert_sand_wavy_C}
   Spec Texture Object=@n49
   TextureSize=AppendVector(A=AppendVector(A=@n50, B=@n50), B=@n50)
   NormalTextureObject=@n51
   World Space Normals=StaticBool() {value=true}
   Export Float 4 Diffuse=StaticBool() {value=true}
  ) {materialFunction=&WorldAlignedTextures_Complex}
  n53 = ComponentMask(None=@n52.Z Diffuse Texture) {b=true, g=true, r=true}
  n54 = ComponentMask(None=@n52.Z Diffuse Texture) {a=true}
  n55 = SetMaterialAttributes(
   Base Color=@n53
   Specular=ComponentMask(None=@n52.Z Spec) {r=true}
   Roughness=ComponentMask(None=@n52.Z Spec) {b=true, desc=Used to be .7}
   Normal=MaterialFunctionCall(
    Normal=Transform(None=@n52.Normal Z Texture) {transformSourceType=TRANSFORMSOURCE_World, transformType=TRANSFORM_Tangent}
    Flatness=Constant() {r=0.6}
   ) {materialFunction=&FlattenNormal}.Result
   Opacity=@n54
  ) {attributeSetTypes [
    #1
    #2
    #3
    #4
    #5
   ]}
  n56 = Constant() {r=0.5}
  n57 = Constant3Vector() {constant=<rgba 0 0 1 0>}
  n58 = SetMaterialAttributes(Base Color=@n53, Specular=@n56, Roughness=Constant() {r=0.7}, Normal=@n57, Opacity=@n54) {attributeSetTypes [
    #1
    #2
    #3
    #4
    #5
   ]}
  n59 = LandscapeLayerCoords() {mappingScale=4}
  n60 = TextureSample(UVs=@n59) {samplerSource=SSM_Wrap_WorldGroupSettings, texture=&T_desert_sand_wavy_C}
  n61 = TextureSample(UVs=@n59, Tex=@n49) {samplerSource=SSM_Wrap_WorldGroupSettings, samplerType=SAMPLERTYPE_Masks, texture=&T_desert_sand_wavy_S}
  n62 = TextureSample(UVs=@n59, Tex=@n51) {
   samplerSource=SSM_Wrap_WorldGroupSettings
   samplerType=SAMPLERTYPE_Normal
   texture=&T_desert_sand_wavy_N
  }
  n63 = SetMaterialAttributes(
   Base Color=@n60.RGB
   Specular=@n61.R
   Roughness=@n61.B
   Normal=MaterialFunctionCall(Normal=@n62.RGB, Flatness=Constant() {r=0.6}) {materialFunction=&FlattenNormal}.Result
   Opacity=@n60.RGB
  ) {attributeSetTypes [
    #1
    #2
    #3
    #4
    #5
   ]}
  n64 = SetMaterialAttributes(Base Color=@n60.RGB, Specular=@n56, Roughness=Constant() {r=0.7}, Normal=@n57, Opacity=@n60.RGB) {attributeSetTypes [
    #1
    #2
    #3
    #4
    #5
   ]}
  n65 = GetMaterialAttributes(None=StaticSwitchParameter(
    True=ShadingPathSwitch(Default=@n55, Mobile=@n58) {desc=Migrated from FeatureLevelSwitch}
    False=ShadingPathSwitch(Default=@n63, Mobile=@n64) {desc=Migrated from FeatureLevelSwitch}
   ) {parameterName=IsStaticmesh}) {attributeGetTypes [#1]}
  n66 = Multiply(A=@n65.Base Color, B=VectorParameter() {parameterName=SandTint, defaultValue=<rgba 1 1 1 1>}.RGB)
  n67 = Multiply(A=VertexColor().G, B=NamedRerouteUsage())
  n68 = LandscapeLayerSample() {parameterName=Mud}
  n69 = GetMaterialAttributes(None=MaterialFunctionCall(
    Bottom=SetMaterialAttributes!(MaterialAttributes=@n48.BlendedMaterial, Opacity=@n48.BlendedMaterial) {attributeSetTypes [#5]}
    Top=SetMaterialAttributes!(MaterialAttributes=@n65.MaterialAttributes, Base Color=@n66) {attributeSetTypes [#1]}
    Alpha=StaticSwitchParameter(True=@n67, False=@n68) {parameterName=IsStaticmesh}
    Height Multiplier=ScalarParameter() {defaultValue=1.5, parameterName=Mud Heightblend Multiplier}
    Contrast=ScalarParameter() {defaultValue=0.5, parameterName=Mud Heightblend Contrast}
   ) {materialFunction=&LayerHeightBlend}.Result) {attributeGetTypes [
    #1
    #2
    #3
    #4
   ]}
  SetMaterialAttributes!(MaterialAttributes, Base Color=@n69.Base Color, Specular=@n69.Base Color, Roughness=@n69.Base Color, Normal=@n69.Base Color) {attributeSetTypes [
    #1
    #2
    #3
    #4
   ]}
 ]
}
