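# Material function for Athena landscape that outputs a height-based sediment color gradient, likely for arid rock banding.
# World-space Z is offset and scaled into UVs for the wrapped GradientMap texture parameter, so colors repeat vertically.
# A quality switch half-desaturates the sample at Low and Medium, leaving full color at High and default.
# The result is lifted by a gray offset, then multiplied by a warm tint.
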MaterialFunction /Game/Athena/Environments/Landscape/MaterialFunctions/MF_Athena_SedimentGradient
graph {nodes [
  n1 = ComponentMask(None=WorldPosition().XYZ) {b=true}
  n2 = Constant() {r=8000}
  n3 = TextureSampleParameter2D(UVs=Divide(A=Add(A=@n1, B=@n2), B=Constant() {r=4000})) {
   group=Arid_Rock
   parameterName=GradientMap
   samplerSource=SSM_Wrap_WorldGroupSettings
   texture=/Game/Environments/Sets/Arid/Textures/GradientSediments__M
  }
  n4 = Constant()
  n5 = Constant() {r=0.5}
  n6 = Desaturation(None=@n3.RGB, Fraction=QualitySwitch(Default=@n4, Low=@n5, High=@n4, Medium=@n5))
  n7 = Constant3Vector() {constant=<rgba 0.5 0.5 0.5 0>}
  FunctionOutput(None=Multiply(A=Add(A=@n6, B=@n7), B=Constant3Vector() {constant=<rgba 0.75 0.6011 0.5586 0>}))
 ]}
