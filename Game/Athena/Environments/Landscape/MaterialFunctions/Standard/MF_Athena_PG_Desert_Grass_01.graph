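# Landscape material function for the desert grass layer of the Athena terrain, outputting material attributes plus Height and plain base color.
# MF_Landscape_PG_DesertGrassDistanceBlend1 supplies grass RGB for color and alpha for height, which also feeds Opacity.
# Squared alpha drives a low specular, while roughness is constant and the normal is flat.
# A grasslands normal map sample and a second GrassDistanceBlend call are left unconnected.
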MaterialFunction /Game/Athena/Environments/Landscape/MaterialFunctions/Standard/MF_Athena_PG_Desert_Grass_01
graph {nodes [
  n1 = MaterialFunctionCall!(UseVertexInterp) {materialFunction=/Game/Athena/Environments/Landscape/MaterialFunctions/MF_Landscape_PG_DesertGrassDistanceBlend1}
  n2 = Constant() {r=0.15}
  n3 = Power(Base=@n1.Grass A, Exp=Constant() {r=2})
  FunctionOutput(None=MakeMaterialAttributes!(
    BaseColor=@n1.Grass RGB
    Metallic
    Specular=LinearInterpolate(B=@n2, Alpha=@n3)
    Roughness=Constant() {r=0.9}
    Anisotropy
    EmissiveColor
    Opacity=FunctionOutput(None=@n1.Grass A) {outputName=Height}
    OpacityMask
    Normal=Constant3Vector() {constant=<rgba 0 0 1 0>}
    Tangent
    WorldPositionOffset
    SubsurfaceColor
    ClearCoat
    ClearCoatRoughness
    AmbientOcclusion
    Refraction
    CustomizedUVs_0
    CustomizedUVs_1
    CustomizedUVs_2
    CustomizedUVs_3
    CustomizedUVs_4
    CustomizedUVs_5
    CustomizedUVs_6
    CustomizedUVs_7
    PixelDepthOffset
    ShadingModel
    Displacement
   ))
  FunctionOutput(None=@n1.Grass RGB) {outputName=BaseColor_Plain}
  MaterialFunctionCall!(UseVertexInterp) {materialFunction=/Game/Athena/Environments/Landscape/MaterialFunctions/MF_Landscape_GrassDistanceBlend}
  TextureSample(UVs=@n1.GrassScale) {
   samplerSource=SSM_Wrap_WorldGroupSettings
   samplerType=SAMPLERTYPE_Normal
   texture=/Game/Packages/DS_Fortnite_Terrain_NoLOD/Textures/T_Grasslands_N
  }
 ]}
