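# Landscape material function for Athena tropical sand, built from the arid wavy sand color, normal and specular textures.
# The default output samples them at input UVs into material attributes with flattened normals, constant roughness and color alpha as height in Opacity, also exposing BaseColor and Height separately.
# A WorldAligned output uses Z-projected color and tangent-converted normal but XY-projected specular, deriving texture size from the same UVs input.
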MaterialFunction /Game/Athena/Environments/Landscape/MaterialFunctions/Standard/MF_Athena_PG_TropicalSand
graph {nodes [
  n1 = FunctionInput(Preview=TextureCoordinate() {uTiling=0.25, vTiling=0.25}) {bUsePreviewValueAsDefault=true, inputName=UVs, inputType=FunctionInput_Vector2}
  n2 = TextureSample(UVs=@n1) {texture=/Game/Environments/Sets/Arid/Textures/T_desert_sand_wavy_C}
  n3 = TextureSample(UVs=@n1) {samplerType=SAMPLERTYPE_Normal, texture=/Game/Environments/Sets/Arid/Textures/T_desert_sand_wavy_N}
  n4 = Constant() {r=0.6}
  FunctionOutput(None=MakeMaterialAttributes!(
    BaseColor=FunctionOutput(None=@n2.RGB) {outputName=BaseColor}
    Metallic
    Specular=TextureSample(UVs=@n1) {samplerType=SAMPLERTYPE_Masks, texture=/Game/Environments/Sets/Arid/Textures/T_desert_sand_wavy_S}.RGB
    Roughness=Constant() {r=0.7}
    Anisotropy
    EmissiveColor
    Opacity=FunctionOutput(None=@n2.A) {outputName=Height, sortPriority=1}
    OpacityMask
    Normal=MaterialFunctionCall(Normal=@n3.RGB, Flatness=@n4) {materialFunction=&FlattenNormal}.Result
    Tangent
    WorldPositionOffset
    SubsurfaceColor
    ClearCoat
    ClearCoatRoughness
    AmbientOcclusion
    Refraction
    CustomizedUVs_0
    CustomizedUVs_1
    CustomizedUVs_2
    CustomizedUVs_3
    CustomizedUVs_4
    CustomizedUVs_5
    CustomizedUVs_6
    CustomizedUVs_7
    PixelDepthOffset
    ShadingModel
    Displacement
   ))
  n5 = MaterialFunctionCall!(
   Diffuse Texture Object=TextureObject() {texture=&T_desert_sand_wavy_C}
   Spec Texture Object=TextureObject() {samplerType=SAMPLERTYPE_Masks, texture=&T_desert_sand_wavy_S}
   TextureSize=AppendVector(A=@n1, B=ComponentMask(None=@n1) {r=true})
   --------------------------
   Use seperate settings for Z projection
   Use seperate texture for Z projection
   Texture Size Z projection
   Z Plane Diffuse Texture Object
   Z Plane Spec Texture Object
   "--------------------------      "
   ProjectionTransitionContrast
   Use Texture Alpha Heightmap to Refine the Transistion
   HeightLerpTransitionContrast
   PickTheDiffuseHeightChannel
   TheDiffuseHeightChannel
   "--------------------------       "
   NormalTextureObject=TextureObject() {samplerType=SAMPLERTYPE_Normal, texture=&T_desert_sand_wavy_N}
   World Space Normals=StaticBool() {value=true}
   "--------------------------  "
   Export Float 4 Diffuse=StaticBool() {value=true}
   Export Float 4 Spec
   "--------------------------     "
   WorldPosition
   World Space Normal
  ) {materialFunction=/Engine/Functions/Engine_MaterialFunctions02/Texturing/WorldAlignedTextures_Complex}
  n6 = Transform(None=@n5.Normal Z Texture) {transformSourceType=TRANSFORMSOURCE_World, transformType=TRANSFORM_Tangent}
  n7 = Constant() {r=0.6}
  FunctionOutput(None=MakeMaterialAttributes(
    BaseColor=ComponentMask(None=@n5.Z Diffuse Texture) {b=true, g=true, r=true}
    Specular=@n5.XY Spec
    Roughness=Constant() {r=0.7}
    Opacity=ComponentMask(None=@n5.Z Diffuse Texture) {a=true}
    Normal=MaterialFunctionCall(Normal=@n6, Flatness=@n7) {materialFunction=&FlattenNormal}.Result
   )) {outputName=WorldAligned, sortPriority=10}
 ]}
