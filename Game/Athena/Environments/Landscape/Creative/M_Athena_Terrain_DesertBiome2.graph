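# Desert biome terrain material for Athena Creative landscapes, also usable on static meshes via the IsStaticmesh switch.
# Noise-varied arid rock, simplified on mobile and lower quality, is tinted, then height-blended with slope-based grass, Forest gravel, Mud/road-edge sand and Gravel layers.
# The switch swaps landscape weights and UVs for slope-masked vertex colours and world-space mapping.
# Grass output spawns desert grass on flat, unmasked ground and jungle foliage from Flowers.
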Material /Game/Athena/Environments/Landscape/Creative/M_Athena_Terrain_DesertBiome2
graph {
 properties {bUseMaterialAttributes=true, pixelDepthOffsetMode=PDOM_Legacy, refractionMethod=RM_IndexOfRefraction, usedShadingModels=Default Lit}
 nodes [
  n1 = VertexNormalWS()
  n2 = Ceil(None=Subtract(A=ComponentMask(None=@n1) {b=true}, B=ScalarParameter() {defaultValue=0.95, parameterName=LS Grass Threshold}))
  n3 = Multiply(A=LandscapeLayerSample() {parameterName=Mud}) {constB=2}
  n4 = LandscapeLayerSample() {parameterName=Road}
  n5 = Subtract(A=Subtract(A=Subtract(A=@n2, B=@n3), B=Multiply(A=@n4) {constB=2}), B=LandscapeLayerSample() {parameterName=NoFoliage})
  n6 = LandscapeLayerSample() {parameterName=Forest}
  MaterialExpressionSubtract_2 = Subtract(A=@n5, B=Multiply(B=@n6) {constA=1.2})
  MaterialExpressionMaterialFunctionCall_6 = MaterialFunctionCall!(Alpha=Multiply(A=@n5, B=@n6), Offset=Constant() {r=0.5}) {materialFunction=/Engine/Functions/Engine_MaterialFunctions01/Opacity/AlphaOffset}
  MaterialExpressionLandscapeLayerSample_4 = LandscapeLayerSample() {parameterName=Flowers}
  LandscapeGrassOutput!(
   Grass=@MaterialExpressionSubtract_2
   Forest=@MaterialExpressionMaterialFunctionCall_6.Result
   Jungle=@MaterialExpressionLandscapeLayerSample_4
  ) {grassTypes [
    Grass {
     grassType=/Game/Athena/Environments/Landscape/Island/Foliage/Athena_Foliage_DesertGrass_CreativeIslands
     input {
      expression=&self:MaterialExpressionSubtract_2
      inputName=None
      mask=0
      maskA=0
      maskB=0
      maskG=0
      maskR=0
      outputIndex=0
     }
    }
    Forest {
     grassType=null
     input {
      expression=&self:MaterialExpressionMaterialFunctionCall_6
      inputName=None
      mask=0
      maskA=0
      maskB=0
      maskG=0
      maskR=0
      outputIndex=0
     }
    }
    Jungle {
     grassType=/Game/Athena/Environments/Landscape/Foliage/Athena_Foliage_Jungle
     input {
      expression=&self:MaterialExpressionLandscapeLayerSample_4
      inputName=None
      mask=0
      maskA=0
      maskB=0
      maskG=0
      maskR=0
      outputIndex=0
     }
    }
   ]}
  n7 = MaterialFunctionCall!(
   AridMasks=TextureObject() {samplerType=SAMPLERTYPE_LinearColor, texture=/Game/Athena/Environments/Landscape/Textures/Arid/T_Athena_Arid_Rock_01_M}
   ProjContrast=Constant() {r=10}
  ) {materialFunction=/Game/Athena/Environments/Landscape/MaterialFunctions/Arid/MF_Athena_Arid_Rock}
  n8 = ScalarParameter() {defaultValue=11348.25, parameterName=NoiseSize}
  n9 = MaterialFunctionCall(
   TextureObject=TextureObject() {samplerType=SAMPLERTYPE_Grayscale, texture=/Game/Global/Textures/Default/Noise/TilingNoise06}
   TextureSize=AppendVector(A=AppendVector(A=@n8, B=@n8), B=@n8)
  ) {materialFunction=&WorldAlignedTexture}
  n10 = Saturate(None=LinearInterpolate(
    A=ScalarParameter() {defaultValue=1, parameterName=NoiseLOW}
    B=ScalarParameter() {defaultValue=-4, parameterName=NoiseHigh}
    Alpha=ComponentMask(None=@n9.XY Texture) {r=true}
   ))
  n11 = GetMaterialAttributes(None=MaterialFunctionCall!(Material=@n7.Mat, NewBaseColor=@n7.Mat, Mask=@n10) {materialFunction=/Engine/Functions/MaterialLayerFunctions/MatLayerBlend_OverrideBaseColor}.Blended Material) {attributeGetTypes [#1]}
  n12 = Constant3Vector() {constant=<rgba 0 0 1 0>}
  n13 = MaterialFunctionCall!(
   In=SetMaterialAttributes!(MaterialAttributes=@n11.MaterialAttributes, Normal=LinearInterpolate(A=@n11.Normal, B=@n12, Alpha=@n10)) {attributeSetTypes [#1]}
   Color=MaterialFunctionCall() {materialFunction=/Game/Athena/Environments/Landscape/MaterialFunctions/MF_Athena_SedimentGradient}.Result
  ) {materialFunction=/Game/Athena/Environments/Landscape/MaterialFunctions/MF_Athena_Material_BaseMultiply}
  n14 = MaterialFunctionCall(In=@n7.MatSimple, Color=Constant3Vector() {desc="Ship hack, Replace Later", constant=<rgba 1 0.7974 0.7166 1>}) {materialFunction=&MF_Athena_Material_BaseMultiply}
  n15 = QualitySwitch(Default=@n13.Result, Low=@n14.Result, High=@n13.Result, Medium=@n14.Result)
  n16 = ShadingPathSwitch(Default=@n15, Deferred=@n15, Mobile=@n14.Result) {desc=Migrated from FeatureLevelSwitch}
  n17 = VectorParameter() {parameterName=BiomeColor, defaultValue=<rgba 0.7656 0.3161 0.162 1>}
  n18 = ComponentMask(None=VertexNormalWS()) {b=true}
  n19 = ScalarParameter() {defaultValue=0.55, group=Rock, parameterName=RockMultiply Z Multiply}
  n20 = Divide(A=ComponentMask(None=WorldPosition().XYZ) {g=true, r=true}, B=ScalarParameter() {defaultValue=700, parameterName=SizeGravelDesert})
  n21 = LandscapeLayerCoords() {mappingScale=4}
  n22 = VertexColor()
  n23 = VertexNormalWS()
  n24 = Saturate(None=LinearInterpolate(
    A=ScalarParameter() {defaultValue=-1.2436, group=Rock, parameterName=CliffBlend}
    B=ScalarParameter() {defaultValue=1, group=Rock, parameterName=CliffBlend2}
    Alpha=ComponentMask(None=@n23) {b=true}
   ))
  n25 = MaterialFunctionCall!(
   Bottom=MaterialFunctionCall(
    Bottom=MaterialFunctionCall!(Material=@n16, Tint=@n17.RGB, TintMask) {materialFunction=/Engine/Functions/MaterialLayerFunctions/MatLayerBlend_Tint}.Blended Material
    Top=MaterialFunctionCall() {materialFunction=/Game/Athena/Environments/Landscape/MaterialFunctions/Standard/MF_Athena_PG_Desert_Grass_01}.Result
    Alpha=Multiply(A=@n18, B=@n19)
    Height Multiplier=ScalarParameter() {defaultValue=0.1, parameterName=GrassRock Height Blend Mult}
    Contrast=ScalarParameter() {defaultValue=5, group=Grass, parameterName=GrassRock Contrast}
   ) {materialFunction=/Game/Athena/Environments/Landscape/MaterialFunctions/Utilities/LayerHeightBlend}.Result
   Top=MaterialFunctionCall!(Coords=StaticSwitchParameter(True=@n20, False=@n21) {parameterName=IsStaticmesh}, Color Tint) {materialFunction=/Game/Athena/Environments/Landscape/MaterialFunctions/Standard/MF_Athena_PG_Gravel_Desert}.Result
   Alpha=StaticSwitchParameter(True=Multiply(A=@n22.R, B=@n24), False=LandscapeLayerSample() {parameterName=Forest}) {parameterName=IsStaticmesh}
   Edge Tint Color=VectorParameter() {parameterName=Forest_Tint, defaultValue=<rgba 1.5 1.5 1.5 0>}.RGB
   Height Multiplier=ScalarParameter() {defaultValue=7, parameterName=Forest Heightblend Multiplier}
   Contrast=ScalarParameter() {defaultValue=0.1, parameterName=Forest Heightblend Contrast}
  ) {materialFunction=&LayerHeightBlend}
  n26 = ScalarParameter() {defaultValue=650, parameterName=SizeSandDesert}
  n27 = MaterialFunctionCall!(UVs=StaticSwitchParameter(True=AppendVector(A=@n26, B=@n26), False=LandscapeLayerCoords() {mappingScale=4}) {parameterName=IsStaticmesh}) {materialFunction=/Game/Athena/Environments/Landscape/MaterialFunctions/Standard/MF_Athena_PG_TropicalSand}
  n28 = MaterialFunctionCall!(
   Material=StaticSwitchParameter(True=@n27.WorldAligned, False=@n27.WorldAligned) {parameterName=IsStaticmesh}
   NewBaseColor=VectorParameter() {parameterName=SandTint, defaultValue=<rgba 0.974 0.5815 0.2333 1>}.RGB
   Mask
  ) {materialFunction=/Engine/Functions/MaterialLayerFunctions/MatLayerBlend_MultiplyBaseColor}
  n29 = VertexColor()
  n30 = LandscapeLayerSample() {parameterName=Road}
  n31 = ScalarParameter() {defaultValue=2, parameterName=RoadMudEdgeBoost}
  n32 = Max(
   A=StaticSwitchParameter(True=Multiply(A=@n29.G, B=@n24), False=LandscapeLayerSample() {parameterName=Mud}) {parameterName=IsStaticmesh}
   B=Subtract(A=Multiply(A=@n30, B=@n31), B=ScalarParameter() {defaultValue=0.7, parameterName=RoadMudEdgeOffset})
  )
  n33 = ScalarParameter() {defaultValue=1.5, parameterName=Mud Heightblend Multiplier}
  n34 = ScalarParameter() {defaultValue=0.5, parameterName=Mud Heightblend Contrast}
  n35 = WorldPosition()
  n36 = StaticSwitchParameter(
   True=Divide(A=ComponentMask(None=@n35.XYZ) {g=true, r=true}, B=ScalarParameter() {defaultValue=700, parameterName=SizeGravel2Desert})
   False=LandscapeLayerCoords() {mappingScale=2}
  ) {parameterName=IsStaticmesh}
  n37 = Multiply(A=VertexColor().B, B=@n24)
  n38 = LandscapeLayerSample() {parameterName=Gravel}
  n39 = GetMaterialAttributes(None=MaterialFunctionCall(
    Bottom=MaterialFunctionCall(Bottom=@n25.Result, Top=@n28.Blended Material, Alpha=@n32, Height Multiplier=@n33, Contrast=@n34) {materialFunction=&LayerHeightBlend}.Result
    Top=MaterialFunctionCall(Coords=@n36) {materialFunction=&MF_Athena_PG_Gravel_Desert}.Result
    Alpha=StaticSwitchParameter(True=@n37, False=@n38) {parameterName=IsStaticmesh}
    Height Multiplier=ScalarParameter() {defaultValue=2, parameterName=Gravel Height}
    Contrast=ScalarParameter() {defaultValue=1, parameterName=Gravel Contrast}
   ) {materialFunction=&LayerHeightBlend}.Result) {attributeGetTypes [
    #2
    #3
    #4
    #5
    #1
   ]}
  SetMaterialAttributes!(
   MaterialAttributes
   Base Color=@n39.Base Color
   Specular=@n39.Base Color
   Roughness=@n39.Base Color
   Emissive Color=@n39.Base Color
   Normal=@n39.Base Color
  ) {attributeSetTypes [
    #2
    #3
    #4
    #5
    #1
   ]}
  VectorParameter() {parameterName=Gravel_MudTint, defaultValue=<rgba 1.5 1.5 1.5 0>}
 ]
}
