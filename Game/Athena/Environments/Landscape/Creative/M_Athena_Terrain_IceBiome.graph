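# Landscape material for the Athena Creative ice biome, layering ice, snow and gravel on terrain with landscape hole masking.
# Ice uses shared winter color, bump-offset swirl noise and a simpler mobile color, height-lerped with snow, then height-blended with full snow by surface slope.
# A world-aligned cave rock section publishes height and normal reroutes.
# Distance-faded grass texture alpha replaces the height in Opacity before the optional painted Gravel layer blends on top.
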Material /Game/Athena/Environments/Landscape/Creative/M_Athena_Terrain_IceBiome
graph {
 properties {bUseMaterialAttributes=true, pixelDepthOffsetMode=PDOM_Legacy, usedShadingModels=Default Lit}
 nodes [
  n1 = TextureObject() {texture=/Game/Packages/DS_Fortnite_Terrain_NoLOD/Textures/T_Cave_Exterior_Wall_D}
  n2 = Constant() {r=1534.9661}
  n3 = StaticBool() {value=true}
  n4 = NamedRerouteUsage()
  n5 = MaterialFunctionCall() {materialFunction=/Game/Athena/Environments/Landscape/MaterialFunctions/Standard/MF_Athena_RoadGravel}
  n6 = AppendVector(A=@n5.Gravel, B=@n5.Gravel)
  n7 = Constant() {r=1.75}
  n8 = ComponentMask(None=Multiply(A=VertexNormalWS(), B=Constant() {r=0.725})) {b=true}
  n9 = ComponentMask(None=MaterialFunctionCall(In=Abs(None=@n8), Contrast=NamedRerouteUsage()) {materialFunction=&CheapContrast}.Result) {r=true}
  n10 = Divide(A=LinearInterpolate(
    A=MaterialFunctionCall(TextureObject=@n1, TextureSize=@n2, Export Float 4=@n3, ProjectionTransitionContrast=@n4) {materialFunction=&WorldAlignedTexture}.XY Texture
    B=Multiply(A=@n6, B=@n7)
    Alpha=@n9
   )) {constB=1.25, desc=Darken Shallow rock to make up for no normal}
  n11 = Constant() {r=4500}
  n12 = NamedRerouteUsage()
  n13 = MaterialFunctionCall(TextureObject=@n1, TextureSize=@n11, Export Float 4=StaticBool() {value=true}, ProjectionTransitionContrast=@n12) {materialFunction=&WorldAlignedTexture}
  n14 = Constant() {r=5}
  n15 = Constant() {r=-4}
  n16 = ComponentMask(None=VertexNormalWS()) {b=true}
  n17 = Clamp(None=Add(A=LinearInterpolate(A=@n14, B=@n15, Alpha=@n16), B=Constant() {r=0.8}))
  n18 = Saturate(None=Divide(A=LinearInterpolate(A=@n10, B=@n13.XY Texture, Alpha=@n17)) {constB=1.5})
  n19 = LinearInterpolate(A=Constant3Vector() {constant=<rgba 0 0 1 0>}, B=@n5.GravelAndMudNormals, Alpha=@n9)
  n20 = MaterialFunctionCall!(
   TextureObject=TextureObject() {samplerType=SAMPLERTYPE_Normal, texture=/Game/Packages/DS_Fortnite_Terrain_NoLOD/Textures/T_Cave_Exterior_Wall_N}
   TextureSize=@n11
   WorldSpace
   Normal
   ProjectionTransitionContrast=Multiply(A=@n12, B=Constant() {r=-1})
   WorldPosition
   Use High Quality Normals
  ) {materialFunction=/Game/Athena/Environments/Materials/WorldAlignedNormalContrast}
  n21 = Constant3Vector() {constant=<rgba 0 0 1 0>}
  n22 = SetMaterialAttributes!(
   MaterialAttributes
   Base Color=ComponentMask(None=@n18) {b=true, g=true, r=true}
   Specular=Constant() {r=0.5}
   Roughness=Constant() {r=0.9}
   Opacity=ComponentMask(None=@n18) {a=true}
   Normal=ShadingPathSwitch(Default=LinearInterpolate(A=@n19, B=@n20.XY Texture, Alpha=@n17), Mobile=LinearInterpolate(A=@n21, B=@n20.XY Texture, Alpha=@n17)) {desc=Migrated from FeatureLevelSwitch}
  ) {attributeSetTypes [
    #1
    #2
    #3
    #4
    #5
   ]}
  n23 = GetMaterialAttributes(None=@n22) {attributeGetTypes [#4, #5]}
  NamedRerouteDeclaration(None=ComponentMask(None=@n23.Opacity) {a=true}) {name=Rock_01_Op, nodeColor=<rgba 1 0 1 1>}
  NamedRerouteDeclaration(None=@n23.Normal) {name=Rock_01_Normal, nodeColor=<rgba 1 0 1 1>}
  NamedRerouteDeclaration(None=Constant() {r=10}) {name=ProjCont, nodeColor=<rgba 1 0 1 1>}
  n24 = MaterialFunctionCall() {materialFunction=/Game/Athena/WinterArt/Materials/MF_SharedWinter}
  n25 = Constant() {r=0.05}
  n26 = LandscapeLayerCoords() {mappingScale=1}
  n27 = TextureSample(UVs=BumpOffset(Coordinate=Multiply(A=@n25, B=@n26), Height=NamedRerouteUsage(), HeightRatioInput=Constant() {r=2})) {
   samplerSource=SSM_Wrap_WorldGroupSettings
   samplerType=SAMPLERTYPE_LinearGrayscale
   texture=/Game/Global/Textures/Default/Noise/T_SwirlNoise_N_height
  }
  n28 = Saturate(None=MaterialFunctionCall!(Cutoff Value=Constant() {r=1.8916}, Lerp Value=Constant() {r=-0.0952}, Gradient=NamedRerouteUsage()) {materialFunction=/Engine/Functions/Engine_MaterialFunctions02/SmoothThreshold}.Result)
  n29 = Multiply(
   A=LinearInterpolate(A=@n24.IceColor, B=Add(A=@n27.R, B=@n24.IceColor), Alpha=@n28)
   B=Constant3Vector() {constant=<rgba 0.7135 0.703 0.6715 1>}
  )
  n30 = Multiply(
   A=Constant3Vector() {constant=<rgba 0.4531 0.7539 1 1>}
   B=MaterialFunctionCall(In=NamedRerouteUsage(), Contrast=Constant() {r=-0.25}) {materialFunction=&CheapContrast}.Result
  )
  n31 = MaterialFunctionCall!(UseLandscapeCoordsOnSwitch) {materialFunction=/Game/Athena/Environments/Landscape/MaterialFunctions/Snow/MF_Athena_Snow_01}
  n32 = Constant() {r=0.6952}
  n33 = Constant() {r=0.75}
  n34 = MaterialFunctionCall(
   Transition Phase=QualitySwitch(Default=@n32, Low=@n33, High=@n32, Medium=@n33)
   Height Texture=NamedRerouteUsage()
   Contrast=Constant() {r=-2.5}
  ) {materialFunction=&HeightLerp}
  n35 = Constant() {r=0.15}
  n36 = Constant() {r=0.2}
  n37 = SetMaterialAttributes!(
   MaterialAttributes
   Base Color=LinearInterpolate(A=ShadingPathSwitch(Default=@n29, Deferred=@n29, Forward=@n29, Mobile=@n30), B=@n31.ViewAngleBaseColor, Alpha=@n34.Alpha)
   Specular=Constant() {r=0.5}
   Roughness=LinearInterpolate(A=LinearInterpolate(A=@n35, B=@n36, Alpha=@n28), B=Constant() {r=0.5}, Alpha=@n34.Alpha)
   Metallic=LinearInterpolate(A=Constant() {r=0.25}, Alpha=@n34.Alpha) {constB=0}
   Normal=LinearInterpolate(A=NamedRerouteUsage(), B=Constant3Vector() {constant=<rgba 0 0 1 0>}, Alpha=@n34.Alpha)
   Opacity=NamedRerouteUsage()
  ) {attributeSetTypes [
    #1
    #2
    #3
    #6
    #5
    #4
   ]}
  n38 = Multiply(A=ComponentMask(None=VertexNormalWS()) {b=true}, B=ScalarParameter() {defaultValue=0.6, group=Rock, parameterName=RockMultiply Z Multiply})
  n39 = ScalarParameter() {defaultValue=0.5, parameterName=GrassRock Height Blend Mult}
  n40 = ScalarParameter() {defaultValue=8, group=Grass, parameterName=GrassRock Contrast}
  n41 = ComponentMask(None=WorldPosition() {worldPositionShaderOffset=WPT_ExcludeAllShaderOffsets}.XYZ) {g=true, r=true}
  n42 = TextureSample(UVs=Divide(A=@n41, B=Constant() {r=700})) {samplerSource=SSM_Wrap_WorldGroupSettings, texture=/Game/Athena/Environments/Landscape/Textures/T_Athena_Grass_Farm_ColorMatched_D_2}
  n43 = TextureSample(UVs=Divide(A=@n41, B=Constant() {r=3000})) {samplerSource=SSM_Wrap_WorldGroupSettings, texture=&T_Athena_Grass_Farm_ColorMatched_D_2}
  n44 = Constant() {r=55000}
  n45 = Constant() {r=45000}
  n46 = ShadingPathSwitch(Default=QualitySwitch(Default=@n44, Low=@n45, High=@n44, Medium=@n44), Mobile=@n45) {desc=Migrated from FeatureLevelSwitch}
  n47 = SetMaterialAttributes!(
   MaterialAttributes=GetMaterialAttributes(None=MaterialFunctionCall!(Bottom=@n37, Top=@n31.Material, Alpha=@n38, Edge Tint Color, Height Multiplier=@n39, Contrast=@n40) {materialFunction=/Game/Athena/Environments/Landscape/MaterialFunctions/Utilities/LayerHeightBlend}.Result).MaterialAttributes
   Opacity=LinearInterpolate(
    A=@n42.A
    B=@n43.A
    Alpha=MaterialFunctionCall!(Distance=@n46) {materialFunction=/Game/Athena/Environments/Landscape/MaterialFunctions/MF_TerrainDistanceFade}.Result
   )
  ) {attributeSetTypes [#4]}
  n48 = Constant3Vector() {constant=<rgba 0.28 0.2437 0.21 1>}
  n49 = MaterialFunctionCall() {materialFunction=&MF_Athena_RoadGravel}
  n50 = Constant3Vector() {constant=<rgba 0.36 0.3902 0.4 1>}
  n51 = Constant() {r=2}
  n52 = SetMaterialAttributes!(
   MaterialAttributes
   Base Color=LinearInterpolate(
    A=Multiply(A=@n48, B=@n49.Gravel_NoColor)
    B=Multiply(A=@n49.Gravel_NoColor, B=@n50)
    Alpha=MaterialFunctionCall(In=@n49.Gravel_NoColor, Contrast=@n51) {materialFunction=&CheapContrast}.Result
   )
   Specular=Constant() {r=0.5}
   Roughness=Constant() {r=1}
   Normal=@n49.GravelAndMudNormals
   Opacity=@n49.GravelAndMudNormals
  ) {attributeSetTypes [
    #1
    #2
    #3
    #5
    #4
   ]}
  n53 = LandscapeLayerSample() {parameterName=Gravel}
  n54 = ScalarParameter() {defaultValue=1, parameterName=Gravel Height}
  n55 = ScalarParameter() {parameterName=Gravel Contrast}
  n56 = GetMaterialAttributes(None=LandscapeLayerSwitch!(
    LayerUsed=MaterialFunctionCall(Bottom=@n47, Top=@n52, Alpha=@n53, Height Multiplier=@n54, Contrast=@n55) {materialFunction=&LayerHeightBlend}.Result
    LayerNotUsed=@n47
   ) {parameterName=Gravel}) {attributeGetTypes [
    #1
    #2
    #3
    #5
   ]}
  n57 = SetMaterialAttributes!(MaterialAttributes, Base Color=@n56.Base Color, Specular=@n56.Base Color, Roughness=@n56.Base Color, Normal=@n56.Base Color) {attributeSetTypes [
    #1
    #2
    #3
    #5
   ]}
  SetMaterialAttributes!(MaterialAttributes=@n57, Opacity Mask=LandscapeVisibilityMask()) {attributeSetTypes [#7]}
 ]
}
