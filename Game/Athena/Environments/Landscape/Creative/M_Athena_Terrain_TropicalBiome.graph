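# Landscape material for the Creative tropical biome terrain, with automatic grass foliage spawning.
# World-aligned volcanic rock and distance-faded, height-tinted grass are height-blended using a slope mask that varies with altitude.
# Optional Forest, Mud (wavy desert sand) and Gravel (beach sand) painted layers blend on top, then feed MF_Curie_Landscape.
# Grass output derives from slope and altitude, minus Forest, Mud, Gravel and NoFoliage layers.
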Material /Game/Athena/Environments/Landscape/Creative/M_Athena_Terrain_TropicalBiome
graph {
 properties {bUseMaterialAttributes=true, pixelDepthOffsetMode=PDOM_Legacy, usedShadingModels=Default Lit}
 nodes [
  n1 = Constant() {r=0.564}
  n2 = Constant() {r=0.4094}
  n3 = ComponentMask(None=WorldPosition().XYZ) {b=true}
  n4 = Saturate(None=Divide(A=@n3, B=Constant() {r=13000}))
  n5 = ComponentMask(None=VertexNormalWS()) {b=true}
  n6 = Multiply(A=LinearInterpolate(A=@n1, B=@n2, Alpha=@n4), B=@n5)
  n7 = Subtract(A=Subtract(A=@n6, B=@n4), B=Constant() {r=0.4524})
  n8 = Ceil(None=Subtract(A=@n5, B=Constant() {r=0.9}))
  n9 = Saturate(None=Divide(A=@n3, B=Constant() {r=15000}))
  n10 = Multiply(A=@n5, B=LinearInterpolate(A=@n1, B=@n2, Alpha=@n9))
  n11 = Saturate(None=Multiply(A=Constant() {r=2}, B=Subtract(A=@n10, B=@n9)))
  n12 = Saturate(None=LinearInterpolate(A=Ceil(None=@n7), B=Multiply(A=@n8, B=@n11), Alpha=ScalarParameter() {parameterName=LS Grass-IsBasic}))
  n13 = Multiply(A=Max(A=LandscapeLayerSample() {parameterName=Mud}, B=LandscapeLayerSample() {parameterName=Gravel})) {constB=2}
  MaterialExpressionSubtract_2 = Subtract(
   A=Subtract(A=Subtract(A=@n12, B=@n13), B=LandscapeLayerSample() {parameterName=NoFoliage})
   B=Multiply(A=LandscapeLayerSample() {parameterName=Forest}) {constA=1.2, constB=1.2}
  )
  LandscapeGrassOutput!(Grass=@MaterialExpressionSubtract_2) {grassTypes [Grass {
     grassType=/Game/Athena/Environments/Landscape/Island/Foliage/Athena_Foliage_Grass_CreativeIslands
     input {
      expression=&self:MaterialExpressionSubtract_2
      inputName=None
      mask=0
      maskA=0
      maskB=0
      maskG=0
      maskR=0
      outputIndex=0
     }
    }]}
  n14 = TextureObject() {texture=/Game/Packages/DS_Fortnite_Terrain_NoLOD/Textures/T_Volcano_Exterior_Wall_Volcanic_AndesiteDarker}
  n15 = Constant() {r=2159.3691}
  n16 = TextureObject() {texture=/Game/Packages/DS_Fortnite_Terrain_NoLOD/Textures/T_Volcano_Exterior_Wall_Volcanic_Andesite}
  n17 = Constant() {r=1}
  n18 = MaterialFunctionCall!(
   TextureObject=@n14
   TextureSize=@n15
   WorldPosition
   Export Float 4=StaticBool() {value=true}
   World Space Normal_Z=Multiply(A=VertexNormalWS(), B=Constant() {r=-0.4254})
   TextureObject_Z=@n16
   World Space Normal
   ProjectionTransitionContrast=@n17
  ) {materialFunction=/Game/Athena/Prototype/Terrain/WorldAlignedTexture_ZTex}
  n19 = Divide(A=@n18.XY Texture, B=Constant() {r=1.6733}) {constB=1.25}
  n20 = Constant() {r=4500}
  n21 = MaterialFunctionCall(
   TextureObject=@n14
   TextureSize=@n20
   Export Float 4=StaticBool() {value=true}
   World Space Normal_Z=VertexNormalWS()
   TextureObject_Z=@n16
   ProjectionTransitionContrast=@n17
  ) {materialFunction=&WorldAlignedTexture_ZTex}
  n22 = Constant() {r=4.868}
  n23 = Constant() {r=-1.7863}
  n24 = ComponentMask(None=VertexNormalWS()) {b=true}
  n25 = Clamp(None=Add(A=LinearInterpolate(A=@n22, B=@n23, Alpha=@n24), B=Constant() {r=0.4369}))
  n26 = Multiply(A=ComponentMask(None=VertexNormalWS()) {b=true}, B=Constant() {r=0.5922})
  n27 = ComponentMask(None=MaterialFunctionCall(In=Abs(None=@n26), Contrast=Constant() {r=20}) {materialFunction=&CheapContrast}.Result) {r=true}
  n28 = LinearInterpolate(
   A=Divide(A=LinearInterpolate(A=@n19, B=@n21.XY Texture, Alpha=@n25)) {constB=1.5}
   B=Divide(A=@n18.Z Texture, B=Constant() {r=3.6224})
   Alpha=@n27
  )
  n29 = TextureObject() {samplerType=SAMPLERTYPE_Normal, texture=/Game/Packages/DS_Fortnite_Terrain_NoLOD/Textures/T_Cave_Exterior_Wall_N}
  n30 = Constant() {r=-50}
  n31 = Constant3Vector() {constant=<rgba 0 0 1 0>}
  n32 = LinearInterpolate(
   A=LinearInterpolate(
    A=MaterialFunctionCall!(TextureObject=@n29, TextureSize=@n15, WorldSpace, Normal, ProjectionTransitionContrast=@n30, WorldPosition, Use High Quality Normals) {materialFunction=/Game/Athena/Environments/Materials/WorldAlignedNormalContrast}.XY Texture
    B=MaterialFunctionCall(TextureObject=@n29, TextureSize=@n20, ProjectionTransitionContrast=@n30) {materialFunction=&WorldAlignedNormalContrast}.XY Texture
    Alpha=@n25
   )
   B=@n31
   Alpha=@n27
  )
  n33 = ComponentMask(None=WorldPosition() {worldPositionShaderOffset=WPT_ExcludeAllShaderOffsets}.XYZ) {g=true, r=true}
  n34 = TextureSample(UVs=Divide(A=@n33, B=Constant() {r=700})) {samplerSource=SSM_Wrap_WorldGroupSettings, texture=/Game/Athena/Environments/Landscape/Textures/T_Athena_Grass_Farm_ColorMatched_D_2}
  n35 = TextureSample(UVs=Divide(A=@n33, B=Constant() {r=3000})) {samplerSource=SSM_Wrap_WorldGroupSettings, texture=&T_Athena_Grass_Farm_ColorMatched_D_2}
  n36 = Constant() {r=55000}
  n37 = Constant() {r=45000}
  n38 = MaterialFunctionCall!(Distance=ShadingPathSwitch(Default=QualitySwitch(Default=@n36, Low=@n37, High=@n36, Medium=@n36), Mobile=@n37) {desc=Migrated from FeatureLevelSwitch}) {materialFunction=/Game/Athena/Environments/Landscape/MaterialFunctions/MF_TerrainDistanceFade}
  n39 = LinearInterpolate(A=@n34.RGB, B=@n35.RGB, Alpha=@n38.Result)
  n40 = VectorParameter() {parameterName=EdgeTintGrass, defaultValue=<rgba 3 0.2487 0.1094 1>}
  n41 = LinearInterpolate(A=@n34.A, B=@n35.A, Alpha=@n38.Result)
  n42 = Constant() {r=2}
  n43 = SetMaterialAttributes!(
   MaterialAttributes
   Base Color=LinearInterpolate(A=@n39, B=Multiply(A=@n40.RGB, B=@n39), Alpha=@n4)
   Specular=LinearInterpolate(B=Constant() {r=0.15}, Alpha=Power(Base=@n41, Exp=@n42))
   Roughness=Constant() {r=0.9}
   Normal=Constant3Vector() {constant=<rgba 0 0 1 0>}
   Opacity=@n41
  ) {attributeSetTypes [
    #1
    #2
    #3
    #4
    #5
   ]}
  n44 = MaterialFunctionCall!(
   Bottom=MakeMaterialAttributes!(
    BaseColor=ComponentMask(None=@n28) {b=true, g=true, r=true}
    Metallic
    Specular=Constant() {r=0.0381}
    Roughness=Constant() {r=0.6}
    Anisotropy
    EmissiveColor
    Opacity=ComponentMask(None=@n28) {a=true}
    OpacityMask
    Normal=ShadingPathSwitch(Default=@n32, Mobile=@n31) {desc=Migrated from FeatureLevelSwitch}
    Tangent
    WorldPositionOffset
    SubsurfaceColor
    ClearCoat
    ClearCoatRoughness
    AmbientOcclusion
    Refraction
    CustomizedUVs_0
    CustomizedUVs_1
    CustomizedUVs_2
    CustomizedUVs_3
    CustomizedUVs_4
    CustomizedUVs_5
    CustomizedUVs_6
    CustomizedUVs_7
    PixelDepthOffset
    ShadingModel
    Displacement
   )
   Top=@n43
   Alpha=@n6
   Edge Tint Color
   Height Multiplier=Constant() {r=0.2381}
   Contrast=Constant() {r=10}
  ) {materialFunction=/Game/Athena/Environments/Landscape/MaterialFunctions/Utilities/LayerHeightBlend}
  n45 = TextureSample(UVs=LandscapeLayerCoords() {mappingScale=4}) {samplerSource=SSM_Wrap_WorldGroupSettings, texture=/Game/Packages/DS_Fortnight/SM/Textures/T_FORT_ForestFloor01}
  n46 = SetMaterialAttributes(
   Base Color=@n45.RGB
   Specular=Constant() {r=0.5}
   Roughness=Constant() {r=1}
   Normal=Constant3Vector() {constant=<rgba 0 0 1 0>}
   Opacity=@n45.RGB
  ) {attributeSetTypes [
    #1
    #2
    #3
    #4
    #5
   ]}
  n47 = LandscapeLayerSwitch!(
   LayerUsed=MaterialFunctionCall(
    Bottom=@n44.Result
    Top=@n46
    Alpha=LandscapeLayerSample() {parameterName=Forest}
    Edge Tint Color=VectorParameter() {parameterName=Forest_Tint, defaultValue=<rgba 1.5 1.5 1.5 0>}.RGB
    Height Multiplier=ScalarParameter() {defaultValue=1, parameterName=Forest Heightblend Multiplier}
    Contrast=ScalarParameter() {defaultValue=0.5, parameterName=Forest Heightblend Contrast}
   ) {materialFunction=&LayerHeightBlend}.Result
   LayerNotUsed=@n44.Result
  ) {parameterName=Forest}
  n48 = LandscapeLayerCoords() {mappingScale=4}
  n49 = TextureSample(UVs=@n48) {samplerSource=SSM_Wrap_WorldGroupSettings, texture=/Game/Environments/Sets/Arid/Textures/T_desert_sand_wavy_C}
  n50 = TextureSample() {samplerSource=SSM_Wrap_WorldGroupSettings, samplerType=SAMPLERTYPE_Masks, texture=/Game/Environments/Sets/Arid/Textures/T_desert_sand_wavy_S}
  n51 = TextureSample(UVs=@n48) {
   samplerSource=SSM_Wrap_WorldGroupSettings
   samplerType=SAMPLERTYPE_Normal
   texture=/Game/Environments/Sets/Arid/Textures/T_desert_sand_wavy_N
  }
  n52 = Constant() {r=0.6}
  n53 = SetMaterialAttributes(
   Base Color=@n49.RGB
   Specular=@n50.R
   Roughness=@n50.R
   Normal=ShadingPathSwitch(
    Default=MaterialFunctionCall(Normal=@n51.RGB, Flatness=@n52) {materialFunction=&FlattenNormal}.Result
    Mobile=Constant3Vector() {constant=<rgba 0 0 1 0>}
   ) {desc=Migrated from FeatureLevelSwitch}
   Opacity=@n49.RGB
  ) {attributeSetTypes [
    #1
    #2
    #3
    #4
    #5
   ]}
  n54 = LandscapeLayerSwitch(
   LayerUsed=MaterialFunctionCall(
    Bottom=@n47
    Top=@n53
    Alpha=LandscapeLayerSample() {parameterName=Mud}
    Height Multiplier=ScalarParameter() {defaultValue=1.5, parameterName=Mud Heightblend Multiplier}
    Contrast=ScalarParameter() {defaultValue=20, parameterName=Mud Heightblend Contrast}
   ) {materialFunction=&LayerHeightBlend}.Result
   LayerNotUsed=@n47
  ) {parameterName=Mud}
  n55 = TextureSample(UVs=LandscapeLayerCoords() {mappingScale=2}) {samplerSource=SSM_Wrap_WorldGroupSettings, texture=/Game/Athena/Environments/Landscape/Textures/T_Athena_BeachSand_01}
  n56 = SetMaterialAttributes!(
   MaterialAttributes
   Base Color=Multiply(A=@n55.RGB, B=VectorParameter() {parameterName=Sand_Tint, defaultValue=<rgba 1.15 1.1441 1.0542 0>}.RGB)
   Roughness=Constant() {r=1}
   Specular=Constant() {r=0.5}
   Normal=Constant3Vector() {constant=<rgba 0 0 1 0>}
   Opacity=@n55.A
  ) {attributeSetTypes [
    #1
    #3
    #2
    #4
    #5
   ]}
  n57 = LandscapeLayerSample() {parameterName=Gravel}
  n58 = ScalarParameter() {defaultValue=1.25, parameterName=Gravel Height}
  n59 = ScalarParameter() {defaultValue=0.01, parameterName=Gravel Contrast}
  n60 = GetMaterialAttributes(None=LandscapeLayerSwitch(
    LayerUsed=MaterialFunctionCall(Bottom=@n54, Top=@n56, Alpha=@n57, Height Multiplier=@n58, Contrast=@n59) {materialFunction=&LayerHeightBlend}.Result
    LayerNotUsed=@n54
   ) {parameterName=Gravel}) {attributeGetTypes [
    #1
    #2
    #3
    #6
    #4
   ]}
  n61 = SetMaterialAttributes!(
   MaterialAttributes
   Base Color=@n60.Base Color
   Specular=@n60.Base Color
   Roughness=@n60.Base Color
   Emissive Color=@n60.Base Color
   Normal=@n60.Base Color
  ) {attributeSetTypes [
    #1
    #2
    #3
    #6
    #4
   ]}
  MaterialFunctionCall!(
   In=@n61
   UseHeightFalloff=StaticBool() {desc=Do Not UseHeightFalloff here}
   OnFireEmissiveOverride
   HeightFalloffRadius
   HeightFalloffHardness
   VariationMasksMobile
   In_Grass
   Is_Stylised_WE
  ) {desc=This should be near very last, materialFunction=/Game/Packages/Fortress_SharedMaterials/Base_Material_Functions/MF_Curie_Landscape}
 ]
}
